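# Looping ambient sound cue for a Helios console control module: a humming bed under randomly timed beeps.
# One of five looping bed waves is picked at random, pitched down, faded in and shaped by a gentle 15-second volume envelope.
# Two looped branches add random beeps after random delays with pitch variation, mixed below the bed.
# All quality levels share this mix, played as an Athena ambient prop with attenuation, owner-limited concurrency and a reverb send.
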SoundCue /Game/Sounds/Helios/Environment_Actors/ConsoleControlModule/Spot_Console_Control_Module_02_Loop_Cue
root=QualityLevel_0
nodes [
 {id=Delay_0, children [Modulator_1], properties {delayMax=3, delayMin=0.65}}
 {id=Delay_1, children [Modulator_2], properties {delayMax=2, delayMin=0.5}}
 {
  id=Enveloper_0
  children [Enveloper_1]
  properties {pitchCurve {editorCurveData {keys [{time=0}]}}, volumeCurve {editorCurveData {keys [<at 0 0>,<at 1 1>]}}}
 }
 {
  id=Enveloper_1
  children [Modulator_0]
  properties {
   loopEnd=15
   pitchCurve {editorCurveData {keys [{arriveTangent=0.0098, interpMode=RCIM_Cubic, time=0}, {interpMode=RCIM_Cubic, time=15}]}}
   volumeCurve {editorCurveData {keys [<at 0 1 RCIM_Cubic>,<at 7.5 0.8 RCIM_Cubic>,<at 15 1>]}}
  }
 }
 {id=Looping_0, children [Delay_0]}
 {id=Looping_1, children [Delay_1]}
 {id=Mixer_0, children [Looping_0,Enveloper_0,Looping_1], properties {inputVolume [0.5,1,0.25]}}
 {id=Mixer_3, children [WavePlayer_16], properties {inputVolume [0.6]}}
 {id=Mixer_4, children [WavePlayer_15], properties {inputVolume [0.6]}}
 {id=Modulator_0, children [Random_3], properties {pitchMax=0.9, pitchMin=0.9}}
 {id=Modulator_1, children [Random_1], properties {pitchMax=1.2, pitchMin=0.9, volumeMin=0.9}}
 {id=Modulator_2, children [Random_2], properties {pitchMax=1.2, pitchMin=0.9}}
 {id=Modulator_3, children [WavePlayer_25], properties {pitchMax=2, pitchMin=2}}
 {id=QualityLevel_0, children [Mixer_0,Mixer_0,Mixer_0,Mixer_0]}
 {id=Random_1, children [Mixer_3, Mixer_4, WavePlayer_17, WavePlayer_19, WavePlayer_20, WavePlayer_22], properties {weights [1,1,1,1,1,1]}}
 {
  id=Random_2
  children [Modulator_3, WavePlayer_26, WavePlayer_27, WavePlayer_28, WavePlayer_29, WavePlayer_31, WavePlayer_23]
  properties {weights [1,1,1,1,1,1,1]}
 }
 {id=Random_3, children [WavePlayer_32, WavePlayer_33, WavePlayer_34, WavePlayer_35, WavePlayer_36], properties {weights [1,1,1,1,1]}}
 {
  id=WavePlayer_15
  properties {soundWaveAssetPtr=&/Spot_Cosole_Control_Module_Beep_02}
 }
 {
  id=WavePlayer_16
  properties {soundWaveAssetPtr=&/Spot_Cosole_Control_Module_Beep_03}
 }
 {
  id=WavePlayer_17
  properties {soundWaveAssetPtr=&/Spot_Cosole_Control_Module_Beep_04}
 }
 {
  id=WavePlayer_19
  properties {soundWaveAssetPtr=&/Spot_Cosole_Control_Module_Beep_06}
 }
 {
  id=WavePlayer_20
  properties {soundWaveAssetPtr=&/Spot_Cosole_Control_Module_Beep_07}
 }
 {
  id=WavePlayer_22
  properties {soundWaveAssetPtr=&/Spot_Cosole_Control_Module_Beep_09}
 }
 {
  id=WavePlayer_23
  properties {soundWaveAssetPtr=&/Spot_Cosole_Control_Module_Beep_01}
 }
 {
  id=WavePlayer_25
  properties {soundWaveAssetPtr=&Spot_Cosole_Control_Module_Beep_03}
 }
 {
  id=WavePlayer_26
  properties {soundWaveAssetPtr=&Spot_Cosole_Control_Module_Beep_04}
 }
 {
  id=WavePlayer_27
  properties {soundWaveAssetPtr=&/Spot_Cosole_Control_Module_Beep_05}
 }
 {
  id=WavePlayer_28
  properties {soundWaveAssetPtr=&Spot_Cosole_Control_Module_Beep_06}
 }
 {
  id=WavePlayer_29
  properties {soundWaveAssetPtr=&Spot_Cosole_Control_Module_Beep_07}
 }
 {
  id=WavePlayer_31
  properties {soundWaveAssetPtr=&Spot_Cosole_Control_Module_Beep_09}
 }
 {
  id=WavePlayer_32
  properties {bLooping=true, soundWaveAssetPtr=&/Spot_Console_Control_Module_Loop_01}
 }
 {
  id=WavePlayer_33
  properties {bLooping=true, soundWaveAssetPtr=&/Spot_Console_Control_Module_Loop_02}
 }
 {
  id=WavePlayer_34
  properties {bLooping=true, soundWaveAssetPtr=&/Spot_Console_Control_Module_Loop_03}
 }
 {
  id=WavePlayer_35
  properties {bLooping=true, soundWaveAssetPtr=&/Spot_Console_Control_Module_Loop_04}
 }
 {
  id=WavePlayer_36
  properties {bLooping=true, soundWaveAssetPtr=&/Spot_Console_Control_Module_Loop_05}
 }
]
properties {
 attenuationSettings=/Game/Sounds/Helios/Environment_Actors/ConsoleTerminals/Spot_Console_Terminal_Communication_Hum_Att
 bOverrideConcurrency=true
 maxDistance=1000
 pitchMultiplier=1.1
 soundClassObject=&Fort_Ambient_Props_Athena
 subtitlePriority=5000
 volumeMultiplier=0.4
 attenuationOverrides {
  bEnableFocusInterpolation=true
  bEnableListenerFocus=true
  bEnableOcclusion=true
  dBAttenuationAtMax=-40
  falloffDistance=512
  focusAzimuth=15
  focusVolumeAttenuation=1.5
  nonFocusAzimuth=45
  nonFocusVolumeAttenuation=0.5
  occlusionInterpolationTime=1
  occlusionLowPassFilterFrequency=10000
  occlusionTraceChannel=ECC_GameTraceChannel11
  occlusionVolumeAttenuation=0.5
  stereoSpread=512
  attenuationShapeExtents=<256 64.5942 90>
 }
 concurrencyOverrides {bLimitToOwner=true, maxCount=2}
 soundSubmixSends [{minSendLevel=1, soundSubmix=&Fortnite_Reverb_Default}]
}
