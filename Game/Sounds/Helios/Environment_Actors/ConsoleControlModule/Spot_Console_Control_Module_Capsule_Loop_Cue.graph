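# Looping ambient SoundCue for a console control module, layering a communication hum with a server rack hum.
# A Mixer blends one of four randomly chosen terminal hum loops with the rack loop, which runs through a Modulator and an Enveloper volume ramp-in.
# All quality levels share this mix.
# Overridden capsule attenuation adds listener focus and occlusion, and concurrency is limited per owner.
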SoundCue /Game/Sounds/Helios/Environment_Actors/ConsoleControlModule/Spot_Console_Control_Module_Capsule_Loop_Cue
root=QualityLevel_0
nodes [
 {
  id=Enveloper_0
  children [Modulator_0]
  properties {pitchCurve {editorCurveData {keys [{time=0}]}}, volumeCurve {editorCurveData {keys [<at 0 0>,<at 1 1>]}}}
 }
 {id=Mixer_0, children [Random_0,Enveloper_0], properties {inputVolume [0.5,0.6]}}
 {id=Modulator_0, children [WavePlayer_0]}
 {id=QualityLevel_0, children [Mixer_0,Mixer_0,Mixer_0,Mixer_0]}
 {id=Random_0, children [WavePlayer_1, WavePlayer_3, WavePlayer_9, WavePlayer_10], properties {weights [1,1,1,1]}}
 {
  id=WavePlayer_0
  properties {bLooping=true, soundWaveAssetPtr=/Game/Sounds/Fort_Prop_Sounds/ServerRacks/Computer_Server_Rack_Tower_Hum_Loop_01}
 }
 {
  id=WavePlayer_1
  properties {bLooping=true, soundWaveAssetPtr=/Game/Sounds/Helios/Environment_Actors/ConsoleTerminals/Spot_Console_Terminal_Communication_Hum_Loop_01}
 }
 {
  id=WavePlayer_10
  properties {bLooping=true, soundWaveAssetPtr=/Game/Sounds/Helios/Environment_Actors/ConsoleTerminals/Spot_Console_Terminal_Communication_Hum_Loop_04}
 }
 {
  id=WavePlayer_3
  properties {bLooping=true, soundWaveAssetPtr=/Game/Sounds/Helios/Environment_Actors/ConsoleTerminals/Spot_Console_Terminal_Communication_Hum_Loop_02}
 }
 {
  id=WavePlayer_9
  properties {bLooping=true, soundWaveAssetPtr=/Game/Sounds/Helios/Environment_Actors/ConsoleTerminals/Spot_Console_Terminal_Communication_Hum_Loop_03}
 }
]
properties {
 attenuationSettings=/Game/Sounds/Helios/Environment_Actors/ConsoleTerminals/Spot_Console_Terminal_Communication_Hum_Att
 bOverrideAttenuation=true
 bOverrideConcurrency=true
 maxDistance=550
 pitchMultiplier=1.1
 soundClassObject=&Fort_Ambient_Props_Athena
 subtitlePriority=5000
 volumeMultiplier=0.301
 attenuationOverrides {
  attenuationShape=Capsule
  bEnableFocusInterpolation=true
  bEnableListenerFocus=true
  bEnableOcclusion=true
  dBAttenuationAtMax=-40
  falloffDistance=350
  focusAzimuth=15
  focusVolumeAttenuation=1.5
  nonFocusAzimuth=45
  nonFocusVolumeAttenuation=0.5
  occlusionInterpolationTime=1
  occlusionLowPassFilterFrequency=10000
  occlusionTraceChannel=ECC_GameTraceChannel11
  occlusionVolumeAttenuation=0.5
  stereoSpread=512
  attenuationShapeExtents=<200 100 90>
 }
 concurrencyOverrides {bLimitToOwner=true, maxCount=2}
}
