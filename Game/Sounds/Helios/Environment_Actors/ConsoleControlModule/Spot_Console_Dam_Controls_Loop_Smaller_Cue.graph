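# Quiet looping ambient SoundCue for a smaller dam controls console, with occlusion, listener focus and a per-owner voice limit.
# A Mixer layers one random pitched-down server hum, with slow looping pitch drift from an Enveloper, over one of four random tech drone hums.
# A short random Delay and volume fade-in envelope precede the mix.
# The first two quality inputs share this chain; the other two are empty.
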SoundCue /Game/Sounds/Helios/Environment_Actors/ConsoleControlModule/Spot_Console_Dam_Controls_Loop_Smaller_Cue
root=QualityLevel_1
nodes [
 {id=Delay_0, children [Mixer_4], properties {delayMax=0.1, delayMin=0.05}}
 {
  id=Enveloper_24
  children [Modulator_4]
  properties {
   bLoop=true
   bLoopIndefinitely=true
   loopEnd=60
   pitchCurve {editorCurveData {keys [{interpMode=RCIM_Cubic, time=0}, <at 30 0.95 RCIM_Cubic>, {interpMode=RCIM_Cubic, time=60}]}}
   volumeCurve {editorCurveData {keys [<at 0 1>]}}
  }
 }
 {
  id=Enveloper_25
  children [Modulator_5]
  properties {
   bLoop=true
   bLoopIndefinitely=true
   loopEnd=60
   pitchCurve {editorCurveData {keys [{interpMode=RCIM_Cubic, time=0}, <at 30 0.95 RCIM_Cubic>, {interpMode=RCIM_Cubic, time=60}]}}
   volumeCurve {editorCurveData {keys [<at 0 1>]}}
  }
 }
 {
  id=Enveloper_26
  children [Modulator_6]
  properties {
   bLoop=true
   bLoopIndefinitely=true
   loopEnd=60
   pitchCurve {editorCurveData {keys [{interpMode=RCIM_Cubic, time=0}, <at 30 0.95 RCIM_Cubic>, {interpMode=RCIM_Cubic, time=60}]}}
   volumeCurve {editorCurveData {keys [<at 0 1>]}}
  }
 }
 {
  id=Enveloper_29
  children [Delay_0]
  properties {
   pitchCurve {editorCurveData {keys [{time=0}]}}
   volumeCurve {editorCurveData {keys [{arriveTangent=3.4293, interpMode=RCIM_Cubic, leaveTangent=3.4293, tangentMode=RCTM_User, time=0, value=0}, <at 0.4 1 RCIM_Cubic>]}}
  }
 }
 {id=Mixer_2, children [Modulator_1], properties {inputVolume [0.8]}}
 {id=Mixer_3, children [WavePlayer_10], properties {inputVolume [1.2]}}
 {id=Mixer_4, children [Random_0,Random_2], properties {inputVolume [1,1]}}
 {id=Modulator_1, children [WavePlayer_8], properties {pitchMax=0.8, pitchMin=0.8}}
 {id=Modulator_4, children [WavePlayer_5], properties {pitchMax=0.9, pitchMin=0.8, volumeMax=0.85, volumeMin=0.8}}
 {id=Modulator_5, children [WavePlayer_6], properties {pitchMax=0.9, pitchMin=0.8}}
 {id=Modulator_6, children [WavePlayer_7], properties {pitchMax=0.9, pitchMin=0.8}}
 {id=QualityLevel_1, children [Enveloper_29,Enveloper_29,null,null]}
 {id=Random_0, children [Enveloper_24,Enveloper_25,Enveloper_26], properties {weights [1,1,1]}}
 {id=Random_2, children [WavePlayer_1,Mixer_2,WavePlayer_9,Mixer_3], properties {weights [1,1,1,1]}}
 {
  id=WavePlayer_1
  properties {bLooping=true, soundWaveAssetPtr=/Game/Sounds/Helios/Environment_Actors/ConsoleWeaponLab/Spot_Console_Tech_Drone_Hum_Loop_01}
 }
 {
  id=WavePlayer_10
  properties {bLooping=true, soundWaveAssetPtr=/Game/Sounds/Helios/Environment_Actors/ConsoleWeaponLab/Spot_Console_Tech_Drone_Hum_Loop_04}
 }
 {id=WavePlayer_5, properties {bLooping=true, soundWaveAssetPtr=/Game/Sounds/Artemis/Environment_Actors/Server/Spot_Server_Stack_Hum_05}}
 {id=WavePlayer_6, properties {bLooping=true, soundWaveAssetPtr=/Game/Sounds/Artemis/Environment_Actors/Server/Spot_Server_Stack_Hum_06}}
 {id=WavePlayer_7, properties {bLooping=true, soundWaveAssetPtr=/Game/Sounds/Artemis/Environment_Actors/Server/Spot_Server_Stack_Hum_07}}
 {
  id=WavePlayer_8
  properties {bLooping=true, soundWaveAssetPtr=/Game/Sounds/Helios/Environment_Actors/ConsoleWeaponLab/Spot_Console_Tech_Drone_Hum_Loop_02}
 }
 {
  id=WavePlayer_9
  properties {bLooping=true, soundWaveAssetPtr=/Game/Sounds/Helios/Environment_Actors/ConsoleWeaponLab/Spot_Console_Tech_Drone_Hum_Loop_03}
 }
]
properties {
 attenuationSettings=/Game/Sounds/Helios/Environment_Actors/ConsoleTerminals/Spot_Console_Terminal_Communication_Hum_Att
 bOverrideAttenuation=true
 bOverrideConcurrency=true
 maxDistance=650
 soundClassObject=&Fort_Ambient_Props_Athena
 subtitlePriority=5000
 volumeMultiplier=0.15
 attenuationOverrides {
  bEnableFocusInterpolation=true
  bEnableListenerFocus=true
  bEnableOcclusion=true
  dBAttenuationAtMax=-40
  falloffDistance=500
  focusAzimuth=15
  nonFocusAzimuth=45
  nonFocusVolumeAttenuation=0.6
  occlusionInterpolationTime=1
  occlusionLowPassFilterFrequency=10000
  occlusionTraceChannel=ECC_GameTraceChannel11
  occlusionVolumeAttenuation=0.5
  attenuationShapeExtents=<150 175 90>
 }
 concurrencyOverrides {bLimitToOwner=true, maxCount=2}
 soundSubmixSends [{minSendLevel=1, soundSubmix=&Fortnite_Reverb_Default}]
}
